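; Prototype POI item spawner that makes a configured list of building actors drop loot when destroyed.
; On BeginPlay it waits for playlist data plus a placeholder delay standing in for a world-loaded event.
; It then adds a BuildingActorLootDropOnDeathComponent to each listed actor and assigns the loot-on-explode package.
; DropChance (random roll) and SpawnItem (tossed pickup) are defined but not called within this Blueprint.
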(event EventBeginPlay
  (Class|FortAsyncActionCurrentPlaylistReady|PlaylistDataReadyAsync
    (:OnReady
      (Utilities|FlowControl|Delay 2.0) ; replace with proper event when the world is fully loaded
      (for _array_element (Variables|Default|GetBuildingActorsListToSpawnItem)
        (bind _returnvalue (AddComponent|NODE_AddBuildingActorLootDropOnDeathComponent-0 :self _array_element))
        (Variables|/Game/Athena/Items/Components/BuildingActorLootDropOnDeathComponent.BuildingActorLootDropOnDeathComponent_C|SetLootOnExplodePackage (Variables|Default|GetLoot%20on%20Explode%20Package) _returnvalue)))))

(fn DropChance
  (bind _lootitemspawnchance (Variables|Default|GetLootItemSpawnChance))
  (if (>= _lootitemspawnchance 1.0)
    (return true)
    (else
      (if (<= (Math|Random|RandomFloatinRange 0.0 1.0) (Math|Float|Clamp(Float) _lootitemspawnchance))
        (return true)
        (else
          (return false))))))

(fn SpawnItem
  (Utilities|IsValid (Variables|Default|GetCurrentActorDestroyed)
    (:Is%20Valid
      (Inventory|SpawnPickupInWorld (Variables|Default|GetItemToSpawn) (Variables|Default|GetItemQuanityToSpawn) "0.000000,0.000000,25.000000" "0.000000,0.000000,1.000000" 25 true true true -1 "Tossed" "Unset" 0 false))))

(fn ConstructionScript)
